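# Physics asset for the chains and medallions on the Urban Phantom medium male player body cosmetic.
# Two chains branch from small base capsules linked to spine_05: one dynamic link to medallion A, two to medallion B.
# Chain and medallion joints use tight swing and twist limits, while base and spine joints rotate freely.
# Large capsules on spine_02 through spine_05 form the torso volume.
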PhysicsAsset /BRCosmetics/Characters/Player/Male/Medium/Bodies/M_MED_UrbanPhantom/Meshes/M_MED_UrbanPhantom_Chains_Physics
bodies {
 base_chain [{shapeType=Capsule, length=0.505, radius=0.505, shapeName=base_chain_capsule, rotation=<rot 0 -90 0>}]
 base_chain_b [{shapeType=Capsule, length=0.505, radius=0.505, shapeName=base_chain_b_capsule, rotation=<rot 0 -90 0>}]
 dyn_chain_a [{shapeType=Capsule, length=3, radius=2, shapeName=dyn_chain_a_capsule, center=<1 0 1.9227>, rotation=<rot 90 0 0>}]
 dyn_chain_b_01 [{shapeType=Capsule, length=3, radius=2, center=<2 0 1.0076>, rotation=<rot 90 0 0>}]
 dyn_chain_b_02 [{shapeType=Capsule, length=3, radius=2, center=<2 0 1.0076>, rotation=<rot 90 0 0>}]
 dyn_medallion_a [{shapeType=Capsule, length=3, radius=2, shapeName=dyn_medallion_a_capsule, center=<1.2778 0 0.9032>, rotation=<rot 74.8637 0 0>}]
 dyn_medallion_b [{shapeType=Capsule, length=3, radius=3, center=<3.6158 0 1.7109>, rotation=<rot 0 0 90>}]
 spine_02 [{shapeType=Capsule, length=14, radius=9, shapeName=spine_02_capsule, center=<2.5578 2.3748 0>}]
 spine_03 [{shapeType=Capsule, length=14, radius=9, shapeName=spine_03_capsule, center=<3.1699 2.4154 0>}]
 spine_04 [{shapeType=Capsule, length=14, radius=9, shapeName=spine_04_capsule, center=<2.015 1.5668 0>}]
 spine_05 [{shapeType=Capsule, length=14, radius=9, shapeName=spine_05_capsule, center=<-1.4651 0.3219 0>}]
}
constraints [
 {
  bone1Name=spine_03
  bone2Name=spine_02
  swing1LimitDegrees=45
  swing1Motion=Free
  swing2LimitDegrees=45
  swing2Motion=Free
  twistLimitDegrees=45
  twistMotion=Free
 }
 {
  bone1Name=spine_04
  bone2Name=spine_03
  swing1LimitDegrees=45
  swing1Motion=Free
  swing2LimitDegrees=45
  swing2Motion=Free
  twistLimitDegrees=45
  twistMotion=Free
 }
 {
  bone1Name=spine_05
  bone2Name=spine_04
  swing1LimitDegrees=45
  swing1Motion=Free
  swing2LimitDegrees=45
  swing2Motion=Free
  twistLimitDegrees=45
  twistMotion=Free
 }
 {
  bone1Name=base_chain
  bone2Name=spine_05
  swing1LimitDegrees=45
  swing1Motion=Free
  swing2LimitDegrees=45
  swing2Motion=Free
  twistLimitDegrees=45
  twistMotion=Free
 }
 {bone1Name=dyn_chain_a, bone2Name=base_chain, swing1LimitDegrees=15, swing2LimitDegrees=15, twistLimitDegrees=3}
 {bone1Name=dyn_medallion_a, bone2Name=dyn_chain_a, swing1LimitDegrees=18, swing2LimitDegrees=10, twistLimitDegrees=3}
 {
  bone1Name=base_chain_b
  bone2Name=spine_05
  swing1LimitDegrees=45
  swing1Motion=Free
  swing2LimitDegrees=45
  swing2Motion=Free
  twistLimitDegrees=45
  twistMotion=Free
 }
 {bone1Name=dyn_chain_b_01, bone2Name=base_chain_b, swing1LimitDegrees=15, swing2LimitDegrees=10, twistLimitDegrees=3}
 {bone1Name=dyn_chain_b_02, bone2Name=dyn_chain_b_01, swing1LimitDegrees=15, swing2LimitDegrees=20, twistLimitDegrees=3}
 {bone1Name=dyn_medallion_b, bone2Name=dyn_chain_b_02, swing1LimitDegrees=15, swing2LimitDegrees=20, twistLimitDegrees=3}
]
